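# Parent material for the NightShade Inverted skeletal-mesh vehicle wheel, giving rim and tire an animated, PaintColor-tinted glow.
# A tiled, panning, fractional UV gradient, gated by Masks channels, drives the glow pattern and perturbs both normal maps.
# Those normals dotted with the tangent-space camera vector form a Fresnel-like rim light.
# Vertex color red blends rim and tire results for color, normal and roughness.
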Material /VehicleCosmetics/Wheels/NightShade_Inverted/Materials/M_WheelParent_NightShadeInverted
graph {
 properties {
  bAutomaticallySetUsageInEditor=false
  bUsedWithSkeletalMesh=true
  previewMesh=/VehicleCosmetics/Wheels/NightShade_Inverted/Meshes/SK_Wheel_NightshadeInverted
  usedShadingModels=Default Lit
 }
 nodes [
  Constant()
  n1 = VectorParameter() {parameterName=PaintColor, defaultValue=<rgba 0.79 0.41 0.17 1>}
  n2 = TextureSampleParameter2D() {
   parameterName=Masks
   samplerType=SAMPLERTYPE_LinearColor
   texture=/VehicleCosmetics/Wheels/NightShade_Inverted/Textures/T_NightshadeInverted_Rim_M
  }
  n3 = Frac(None=Panner(Coordinate=TextureCoordinate() {uTiling=24, vTiling=24}) {speedY=1})
  n4 = Add(A=ComponentMask(None=@n3) {r=true}, B=ComponentMask(None=@n3) {g=true})
  n5 = ConstantBiasScale(None=Multiply(A=@n2.R, B=@n4)) {bias=-1, scale=2}
  n6 = Clamp(None=Multiply(A=@n5, B=ScalarParameter() {defaultValue=3, parameterName=EmissiveIntensity})) {maxDefault=12}
  n7 = TextureSampleParameter2D() {
   parameterName=Normal_Rim
   samplerType=SAMPLERTYPE_Normal
   texture=/VehicleCosmetics/Wheels/NightShade_Inverted/Textures/T_NightshadeInverted_Rim_N
  }
  n8 = Add(A=Clamp(None=@n5) {maxDefault=12}, B=@n7.RGB)
  n9 = DotProduct(A=@n8, B=Transform(None=CameraVectorWS()) {transformSourceType=TRANSFORMSOURCE_World, transformType=TRANSFORM_Tangent})
  n10 = Constant() {r=1.1}
  n11 = Constant() {r=0.3}
  n12 = Multiply(A=OneMinus(None=Multiply(A=@n9, B=@n10)), B=Multiply(A=@n11, B=@n1.RGB))
  n13 = ScalarParameter() {defaultValue=1, parameterName=GlowIntensity}
  n14 = ConstantBiasScale(None=Multiply(A=@n4, B=@n2.B)) {bias=-1, scale=3}
  n15 = TextureSampleParameter2D() {
   parameterName=Normal_Tier
   samplerType=SAMPLERTYPE_Normal
   texture=/VehicleCosmetics/Wheels/NightShade_Inverted/Textures/T_NightshadeInverted_Tire_N
  }
  n16 = Add(A=@n14, B=@n15.RGB)
  n17 = DotProduct(A=@n16, B=Transform(None=CameraVectorWS()) {transformSourceType=TRANSFORMSOURCE_World, transformType=TRANSFORM_Tangent})
  n18 = Constant() {r=1.5}
  n19 = LinearInterpolate(A=OneMinus(None=Multiply(A=@n17, B=@n18)), B=@n14, Alpha=@n2.B)
  n20 = Multiply(A=@n11, B=@n1.RGB)
  n21 = VertexColor()
  LinearInterpolate(
   A=LinearInterpolate(A=Multiply(A=@n1.RGB, B=@n6), B=Multiply(A=@n12, B=@n13), Alpha=Constant() {r=0.3})
   B=LinearInterpolate(A=Multiply(A=@n19, B=@n20), B=@n1.RGB, Alpha=Constant() {r=0.12})
   Alpha=@n21.R
  )
  LinearInterpolate(A=@n8, B=@n16, Alpha=@n21.R)
  n22 = ScalarParameter() {defaultValue=1, parameterName=Roughness}
  LinearInterpolate(A=Constant() {r=1}, B=@n22, Alpha=@n21.R)
  n23 = LightVector()
  Multiply(A=Clamp(None=DotProduct(A=@n7.RGB, B=@n23)))
  Multiply(B=@n22)
 ]
}
